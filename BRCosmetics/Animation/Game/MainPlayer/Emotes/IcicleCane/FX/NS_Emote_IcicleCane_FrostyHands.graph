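# Niagara system for the Icicle Cane emote: frost mist and sparkles on the player's hands, plus a delayed icy impact.
# Continuous emitters sample skeletal mesh bones, with user parameters tuning hand-mist rate, life and alpha.
# Bursts about 1.3 seconds in add body mist, ground fog, snowflakes and a slope-aligned, trace-placed frost impact.
# Spawn groups drive per-layer settings and visibility tags, while time-of-day scalars adjust brightness and opacity.
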NiagaraSystem /BRCosmetics/Animation/Game/MainPlayer/Emotes/IcicleCane/FX/NS_Emote_IcicleCane_FrostyHands
dependencies {
 usedDataInterfaces [
  &NiagaraDataInterfaceCurve
  &NiagaraDataInterfaceColorCurve
  &NiagaraDataInterfaceArrayInt32
  &NiagaraDataInterfaceCollisionQuery
  &NiagaraDataInterfaceDebugDraw
  &NiagaraDataInterfaceSpriteRendererInfo
  &NiagaraDataInterfaceSkeletalMesh
 ]
 usedDynamicInputs [&SimulationPosition, &MakeVector4_FromLinearColor]
 usedModules [
  &self:SystemSpawnScript.SystemScriptSource.SystemScriptGraph.NiagaraNodeAssignment_0.SetVariables_#1
  &self:SystemSpawnScript.SystemScriptSource.SystemScriptGraph.NiagaraNodeAssignment_1.SetVariables_#2
  &SystemState
  &self:SystemSpawnScript.SystemScriptSource.SystemScriptGraph.NiagaraNodeAssignment_2.SetVariables_#3
 ]
}
emitters {
 Frost_Impact_ground002 {
  emitterUpdateScript [
   &EmitterState(Life Cycle Mode=::NewEnumerator1, Loop Behavior=::NewEnumerator1, Loop Delay=1.32, Loop Duration=1.5, UseLoopDelay=true)
   &SpawnBurst_Instantaneous(Spawn Count=5, Use Spawn Probability=false, UseLoopCountLimit=false) {module=SpawnBurst_Instantaneous001}
   &SpawnBurst_Instantaneous(Spawn Group=1, Use Spawn Probability=false, UseLoopCountLimit=false)
   &SpawnBurst_Instantaneous(Spawn Count=2, Spawn Group=4, Use Spawn Probability=false, UseLoopCountLimit=false) {module=SpawnBurst_Instantaneous002}
  ]
  particleSpawnScript [
   /SetVariables(Particles.VisibilityTag=&SpawnGroup()) {isSetParameters=true}
   &InitializeParticle(
    Color Mode=::NewEnumerator0
    Lifetime=&MergeHelper_FloatSwitch(0=0.35, 1=0.4, 2=?, 3=?, Index=&SpawnGroup(), Num=::NewEnumerator0)
    Mesh Scale Mode=::NewEnumerator3
    Position Offset=&MergeHelper_VectorSwitch(0=<0 0 0>, 1=<0 0 0>, 2=?, 3=?, Index=&SpawnGroup(), Num=::NewEnumerator0)
    Sprite Size Mode=::NewEnumerator3
    Uniform Sprite Size=&MergeHelper_FloatSwitch(0=300, 1=120, 2=?, 3=?, Index=&SpawnGroup(), Num=::NewEnumerator0)
    UsePositionOffset=false
    Write Material Random=true
   )
   /SetVariables(
    Particles.1 dynparam:Vector4f=&MakeVector4(X=0, Y=&RandomRangeFloat(Minimum=0, Maximum=4), Z=0, W=2.5)
    Particles.SpriteAlignment=&TransformVector(Vector=<0 0 0>, Destination Space=::World)
   ) {isSetParameters=true}
   &SkeletalMeshLocation(
    MaskSpawnGroup=false
    Mesh Sampling Type=::NewEnumerator5
    Orientation Sampling=::NewEnumerator1
    Position Sampling=::NewEnumerator2
    Sampled Position Offset Coordinate Space:ENiagaraSkelMeshTransforms=::NewEnumerator2
    Skeletal Mesh=$System.Skeletal Mesh
    Socket Sampling Mode=::NewEnumerator5
   ) {module=SkeletalMeshLocation001}
   &CameraOffset(Camera Offset Amount=&MergeHelper_FloatSwitch(0=0, 1=3, 2=30, 3=0, 4=0, Index=&SpawnGroup(), Num=::NewEnumerator3))
   &InitialMeshOrientation(
    Mesh Axis To Orient=<0 0 1>
    Mesh Orientation Mode=::NewEnumerator2
    Orientation Vector=<0 0 1>
    Rotation=&RandomRangeVector(Minimum=<0 0 0>, Maximum=<0 0 1>)
    Use Rotation Vector=true
   )
   /SetVariables(
    Particles.feet SpriteFacing:Vector3f=<0 0 1>
    Particles.feet sprite Position:NiagaraPosition=&AddVectorToPosition(Position=$Particles.Position, Vector=<0 0 12>)
   ) {isSetParameters=true}
  ]
  particleUpdateScript [
   /SetVariables(
    Particles.Glass DynParam:Vector4f=&MakeVector4(
     X=&FloatFromCurve(FloatCurve=DataInterface() {propertyValues {Curve {keys [{arriveTangent=0.6769, time=0, value=0.2}, <at 1 0.5>]}}})
     Y=&FloatFromCurve(
      FloatCurve=DataInterface() {propertyValues {Curve {keys [{arriveTangent=-0.5818, interpMode=RCIM_Cubic, leaveTangent=-0.5818, tangentMode=RCTM_User, time=0, value=0.5}, <at 1 0.5>]}}}
      Scale Curve=0.5
     )
     Z=&FloatFromCurve(FloatCurve=DataInterface() {propertyValues {Curve {keys [<at 0 0>,<at 0.15 1>,<at 1 1.2>]}}})
     W=0
    )
    Particles.Glow Dynamic params:Vector4f=&MakeVector4(X=2, Y=0, Z=0, W=0)
    Particles.Main Star DynParams:Vector4f=&MakeVector4(
     X=&FloatFromCurve(FloatCurve=DataInterface() {propertyValues {Curve {keys [
          <at 0 0.05>
          {arriveTangent=2.6196, interpMode=RCIM_Cubic, leaveTangent=2.6196, tangentMode=RCTM_User, time=0.5, value=0.05}
          {interpMode=RCIM_Cubic, tangentMode=RCTM_User, time=0.9, value=1}
         ]}}})
     Y=&RandomRangeFloat(Minimum=0, Maximum=1)
     Z=0
     W=0
    )
   ) {isSetParameters=true}
   /SetVariables(
    Particles.Ground Glass Linear Color:LinearColor=&MakeLinearColorFromVectorAndFloat(
     "Vector (RGB)"=&MakeVectorFromLinearColorRGB(Color=&ColorFromCurve(ColorCurve=DataInterface() {propertyValues {
         AlphaCurve {keys [<at 0 0>,<at 0.5052 0.5267>,<at 1 0>]}
         BlueCurve {keys [<at 0 5>,<at 1 1>]}
         GreenCurve {keys [<at 0 4.0361>,<at 1 0.8072>]}
         RedCurve {keys [<at 0 3.1927>,<at 1 0.6385>]}
        }}))
     "Float (Alpha)"=&FloatFromCurve(
      FloatCurve=DataInterface() {propertyValues {Curve {keys [{interpMode=RCIM_Cubic, tangentMode=RCTM_User, time=0.5, value=1}, {interpMode=RCIM_Cubic, tangentMode=RCTM_User, time=1, value=0}]}}}
      Scale Curve=&FN_FrontEndFloatModifier(
       Gameplay=&FNTimeOfDayScalar(Night Scale=5, Morning Scale=1, Day Scale=3.5, Evening Scale=1)
       Front End=0.5
      )
     )
    )
    Particles.Ground glass scale:Vector3f=&VectorFromFloat(Value=&FloatFromCurve(
      FloatCurve=DataInterface() {propertyValues {Curve {keys [
          {interpMode=RCIM_Cubic, tangentMode=RCTM_Break, time=0, value=0.2}
          {arriveTangent=2.9365, interpMode=RCIM_Cubic, leaveTangent=2.9365, tangentMode=RCTM_User, time=0.1, value=1}
          {interpMode=RCIM_Cubic, tangentMode=RCTM_Break, time=1, value=2}
         ]}}}
      Scale Curve=0.65
     ))
   ) {isSetParameters=true}
   &Drag(
    Drag=&MergeHelper_FloatSwitch(
     0=0
     1=0
     2=0
     3=0
     4=&FloatFromCurve(
      FloatCurve=DataInterface() {propertyValues {Curve {defaultValue=1, keys [{tangentMode=RCTM_User, time=0, value=0}, {tangentMode=RCTM_User, time=1, value=1}]}}}
      Scale Curve=60
     )
     Index=&SpawnGroup()
     Num=::NewEnumerator3
    )
    Use Linear Drag=true
    Use Rotational Drag=true
   )
   &SolveForcesAndVelocity(Clamp Velocity=false, Limit Acceleration=false)
   &ParticleState()
   &self:Slope_Position_03(
    Apply Position=true
    Apply Rotation=true
    Height Adjustment/Offset=2
    Slope Position:NiagaraPosition=$Particles.SkeletalMeshLocation001.SampledPosition
    Trace Length=120
    Trace Origin Height=85
   ) {module=Slope_Position_03}
   /SetVariables(
    Particles.Mesh Linear Color:LinearColor=&MakeLinearColorFromVectorAndFloat(
     "Vector (RGB)"=&MakeVectorFromLinearColorRGB(Color=&ColorFromCurve(ColorCurve=DataInterface() {propertyValues {
         AlphaCurve {keys [<at 0.729 0.7 RCIM_Cubic>,<at 1 0 RCIM_Cubic>]}
         BlueCurve {keys [<at 0 0.6198 RCIM_Cubic>]}
         GreenCurve {keys [<at 0 0.5249 RCIM_Cubic>]}
         RedCurve {keys [<at 0 0.4087 RCIM_Cubic>]}
        }}))
     "Float (Alpha)"=&FloatFromCurve(
      FloatCurve=DataInterface() {propertyValues {Curve {
         defaultValue=1
         keys [
          {arriveTangent=9.0725, interpMode=RCIM_Cubic, leaveTangent=9.0725, tangentMode=RCTM_User, time=0, value=0}
          {arriveTangent=-2.5249, interpMode=RCIM_Cubic, leaveTangent=-2.5249, tangentMode=RCTM_User, time=0.1408, value=0.9847}
          {arriveTangent=-0.0405, interpMode=RCIM_Cubic, leaveTangent=-0.0405, tangentMode=RCTM_User, time=1, value=0}
         ]
        }}}
      Scale Curve=&FNTimeOfDayScalar(Night Scale=0.3, Morning Scale=0.3, Day Scale=0.5, Evening Scale=0.25)
     )
    )
    Particles.mesh pos:NiagaraPosition=&AddVectorToPosition(Position=$Particles.Position, Vector=<0 0 0>)
   ) {isSetParameters=true}
   /SetVariables(
    Particles.Feet glow size:Vector2f=&Multiply_Vector2D_ByFloat(
     Float=&FloatFromCurve(
      FloatCurve=DataInterface() {propertyValues {Curve {defaultValue=1, keys [{tangentMode=RCTM_User, time=0, value=0}, {tangentMode=RCTM_User, time=1, value=1}]}}}
      Scale Curve=2
     )
     Vector2D=$Particles.SpriteSize
    )
    Particles.feet glow color:LinearColor=&MakeLinearColorFromVectorAndFloat(
     "Vector (RGB)"=&MakeVectorFromLinearColorRGB(Color=&ColorFromCurve(ColorCurve=DataInterface() {propertyValues {
         AlphaCurve {keys [<at 0 1>,<at 0.8001 0>]}
         BlueCurve {keys [<at 0 1>,<at 0.5062 0.9905>]}
         GreenCurve {keys [<at 0 1>,<at 0.5062 1>]}
         RedCurve {keys [<at 0 1>,<at 0.5062 0.6419>]}
        }}))
     "Float (Alpha)"=&FloatFromCurve(
      FloatCurve=DataInterface() {propertyValues {Curve {keys [<at 0 1>,<at 1 0>]}}}
      Scale Curve=&FN_FrontEndFloatModifier(
       Gameplay=&FNTimeOfDayScalar(Night Scale=0.1, Morning Scale=0.1, Day Scale=0.15, Evening Scale=0.15)
       Front End=1
      )
     )
    )
   ) {isSetParameters=true}
   /SetVariables(Particles.Corona_Size:Vector2f=&Multiply_Vector2D_ByFloat(
     Float=&FloatFromCurve(
      FloatCurve=DataInterface() {propertyValues {Curve {defaultValue=1, keys [{tangentMode=RCTM_User, time=0, value=0}, {tangentMode=RCTM_User, time=1, value=1}]}}}
      Scale Curve=1.5
     )
     Vector2D=$Particles.SpriteSize
    )) {isSetParameters=true}
  ]
  rendererClasses [&NiagaraSpriteRendererProperties, &NiagaraMeshRendererProperties]
  renderers [
   {rendererClass=&NiagaraSpriteRendererProperties}
   {rendererClass=&NiagaraMeshRendererProperties, rendererIndex=1}
  ]
 }
 Frosty_Mist_body_Burst {
  emitterSpawnScript [/SetVariables(Emitter.dust spec intensity:NiagaraFloat=&RandomRangeFloat(Minimum=0.1, Maximum=0.2)) {isSetParameters=true}]
  emitterUpdateScript [
   &EmitterState(
    Inactive Response=::NewEnumerator1
    Life Cycle Mode=::NewEnumerator1
    Loop Behavior=::NewEnumerator1
    Loop Delay=1.35
    Loop Duration=1
    UseLoopDelay=true
   )
   &SpawnBurst_Instantaneous(Spawn Count=16, Use Spawn Probability=false, UseLoopCountLimit=false)
  ]
  particleSpawnScript [
   &PartitionParticles(Number Of Partitions=16) {module=PartitionParticles001}
   &InitializeParticle(
    Color=$System.TOD Snowy Mist
    Lifetime=0.57
    Mass Mode=::NewEnumerator1
    Sprite Rotation Mode=::NewEnumerator2
    Sprite Size=&Vector2DFromFloat(Value=&RandomRangeFloat(Minimum=30, Maximum=40))
    Sprite Size Mode=::NewEnumerator4
    UsePositionOffset=false
    Write Material Random=true
   ) {module=InitializeParticle001}
   &CameraOffset(Camera Offset Amount=-20) {enabled=false}
   &ShapeLocation(MaskSpawnGroup=false, Override Seed=false, Sphere Radius=10, UseOwnerScale=true)
   &ApplyInitialForces()
  ]
  particleUpdateScript [
   &ParticleState()
   &GravityForce(Gravity=<0 0 -150>) {enabled=false}
   &SkeletalMeshLocation(
    Bone / Socket Index=&SelectIntFromArray(
     Direct Array Index=$Particles.PartitionParticles001.Partition
     Int Selection Array=DataInterface() {propertyValues {IntData [0,1,2,3,4,5,6,7,8,9,10,11,12,13,14,15]}}
    )
    Bone Sampling Mode=::NewEnumerator4
    MaskSpawnGroup=false
    Position Sampling=::NewEnumerator1
    Skeletal Mesh=$System.body skeletal mesh
    Source Movement Scale:Vector3f=<0 0 0>
   )
   &Drag(Drag=2.5, Use Linear Drag=true, Use Rotational Drag=true)
   &InheritSourceMovement(ApplyForce=false, ApplyPosition=true, ApplyPositionFalloff=false, Source Velocity:Vector3f=$Emitter.InheritVelocityBlend)
   &InheritVelocity(Inherited Velocity Speed Limit=250, Source Speed Threshold=50, Speed Limit=true, UseSpeedThreshold=true)
   &PointAttractionForce(
    Attraction Radius=5
    Attraction Strength=1500
    Attractor Position:NiagaraPosition=$Particles.Initial.ShapeLocation.ShapePosition
    Kill Radius Overshoot Correction=5
    Kill Within Radius=true
    Module.Use Falloff=true
    Position To Attract:NiagaraPosition=$Particles.SkeletalMeshLocation.SampledPosition
    UseAttractorOffset=false
   )
   &SolveForcesAndVelocity(Clamp Velocity=false, Limit Acceleration=false)
   &SubUVAnimation(
    End Frame=25
    Sub UV Play Rate:NiagaraFloat=&Multiply_Float(A=0.1, B=$Particles.NormalizedAge)
    SubUV Animation Mode:ENiagara_SubUVLookupMode=::NewEnumerator3
   ) {module=SubUVAnimation001}
   &SpriteRotationRate(Rotation Rate=&FloatFromCurve(
     FloatCurve=DataInterface() {propertyValues {Curve {keys [<at 0 0 RCIM_Cubic>,<at 0.3 1 RCIM_Cubic>,<at 1 0>]}}}
     Scale Curve=&RandomRangeFloat(Minimum=-20, Maximum=20)
    ))
   &ScaleSpriteSize(Scale Factor=&Vector2DFromFloat(Value=&FloatFromCurve(
      FloatCurve=DataInterface() {propertyValues {Curve {
         defaultValue=1
         keys [{interpMode=RCIM_Cubic, tangentMode=RCTM_User, time=0, value=0.5}, {interpMode=RCIM_Cubic, tangentMode=RCTM_User, time=0.15, value=1.5}]
        }}}
      Scale Curve=1.7
     )))
   &ScaleColor(
    Scale Alpha=&FloatFromCurve(
     FloatCurve=DataInterface() {propertyValues {Curve {
        defaultValue=1
        keys [
         {interpMode=RCIM_Cubic, tangentMode=RCTM_User, time=0, value=0}
         {interpMode=RCIM_Cubic, tangentMode=RCTM_User, time=0.0553, value=0.9718}
         {tangentMode=RCTM_User, time=0.7, value=0}
        ]
       }}}
     Scale Curve=&Multiply_Float(
      A=&FNTimeOfDayScalar(Night Scale=0.5, Morning Scale=0.5, Day Scale=0.5, Evening Scale=0.35)
      B=1
     )
    )
    ScaleA=true
    ScaleRGB=true
   ) {module=ScaleColor001}
   &BlendSpace("Position Blend (World <-> Local)"=0.5, Position Mode=2, Rotation Contribution=0.5, Translation Contribution=0.5, Velocity Mode=1) {enabled=false}
   &DynamicMaterialParameters(
    Index 0 Param 1=0.5
    Module.Index0Float0Write=true
    Module.Index0Float1Write=true
    Module.Index0Float2Write=true
    Module.Index0Float3Write=true
   ) {enabled=false}
   &DynamicMaterialParameters(
    Index 0 Param 1=&FloatFromCurve(FloatCurve=DataInterface() {propertyValues {Curve {keys [<at 0 0 RCIM_Cubic>,<at 1.0011 0.2>]}}}, Scale Curve=0.05)
    Index 1 Param 1=&FloatFromCurve(
     FloatCurve=DataInterface() {propertyValues {Curve {keys [<at 0 0>,<at 0.25 1>]}}}
     Scale Curve=&RandomRangeFloat(Minimum=2, Maximum=1)
    )
    Index 1 Param 2=&FloatFromCurve(
     FloatCurve=DataInterface() {propertyValues {Curve {keys [
         {interpMode=RCIM_Cubic, leaveTangent=3.7422, tangentMode=RCTM_Break, time=0, value=0.2}
         {interpMode=RCIM_Cubic, leaveTangent=0.9066, time=1, value=2}
        ]}}}
     Scale Curve=&RandomRangeFloat(Minimum=0.5, Maximum=1)
    )
    Index 1 Param 3=&FloatFromCurve(FloatCurve=DataInterface() {propertyValues {Curve {keys [
         {interpMode=RCIM_Cubic, leaveTangent=2.8445, tangentMode=RCTM_Break, time=0, value=0.2}
         {interpMode=RCIM_Cubic, leaveTangent=63.2978, time=0.25, value=1}
        ]}}})
    Index 1 Param 4=&FloatFromCurve(
     FloatCurve=DataInterface() {propertyValues {Curve {keys [<at 0 1>,<at 1 0.8>]}}}
     Scale Curve=&RandomRangeFloat(Minimum=0.5, Maximum=1)
    )
    Index 2 Param 1=&FloatFromCurve(FloatCurve=DataInterface() {propertyValues {Curve {keys [{interpMode=RCIM_Cubic, leaveTangent=3.2309, tangentMode=RCTM_Break, time=0.1, value=0}, <at 1 1>]}}})
    Index 2 Param 2=$Particles.MaterialRandom
    Index 2 Param 3=&FloatFromCurve(FloatCurve=DataInterface() {propertyValues {Curve {keys [
         {interpMode=RCIM_Cubic, tangentMode=RCTM_Break, time=0, value=1}
         {interpMode=RCIM_Cubic, leaveTangent=11.1233, tangentMode=RCTM_Break, time=1, value=2}
        ]}}})
    Index 2 Param 4=0.75
    Module.Index0Param0Write=true
    Module.Index0Param1Write=true
    Module.Index0Param2Write=true
    Module.Index0Param3Write=true
    Module.Index1Param0Write=true
    Module.Index1Param1Write=true
    Module.Index1Param2Write=true
    Module.Index1Param3Write=true
    Module.Index2Param0Write=true
    Module.Index2Param1Write=true
    Module.Index2Param2Write=true
    Module.Index2Param3Write=true
    Write Parameter Index 1=true
    Write Parameter Index 2=true
   ) {module=DynamicMaterialParameters001}
  ]
  rendererClasses [&NiagaraSpriteRendererProperties]
  renderers [{rendererClass=&NiagaraSpriteRendererProperties}]
 }
 Frosty_Mist_hands {
  emitterUpdateScript [
   &EmitterState()
   &SpawnRate(
    SpawnRate=&FloatFromCurve(
     FloatCurve=DataInterface() {propertyValues {Curve {keys [<at 0 1.3>,<at 0.3916 1.3>,<at 0.5968 3>,<at 0.9315 3>,<at 1 2.5>]}}}
     CurveIndex=$Emitter.NormalizedLoopAge
     Scale Curve=&Multiply_Float(A=80, B=$User.FrostyMistTrail_Rate)
    )
    Use Spawn Probability=false
   )
   &SpawnRate(SpawnRate=&Multiply_Float(A=15, B=$User.FrostyMistTrail_Rate), Use Spawn Probability=false) {enabled=false, module=SpawnRate001}
   &SpawnPerUnit(Max Movement Threshold=1, Module.Use Max Movement Threshold=true, Module.Use Movement Tolerance=true, Use Spawn Probability=false)
  ]
  particleSpawnScript [
   &PartitionParticles() {module=PartitionParticles001}
   &InitializeParticle(
    Color=$System.TOD Snowy Mist
    Lifetime=$User.MistyTrail_Life
    Mass Mode=::NewEnumerator1
    Sprite Rotation Mode=::NewEnumerator2
    Sprite Size=&Vector2DFromFloat(Value=&RandomRangeFloat(Minimum=10, Maximum=20))
    Sprite Size Mode=::NewEnumerator4
    UsePositionOffset=false
    Write Material Random=true
   ) {module=InitializeParticle001}
   &CameraOffset(Camera Offset Amount=10)
   &ShapeLocation(MaskSpawnGroup=false, Override Seed=false, Sphere Radius=10, UseOwnerScale=true)
   &SkeletalMeshLocation(
    Bone / Socket Index=$Particles.PartitionParticles001.Partition
    Bone Sampling Mode=::NewEnumerator4
    MaskSpawnGroup=false
    Position Sampling=::NewEnumerator1
    Sampled Position Offset=&MergeHelper_VectorSwitch(0=<5 0 0>, 1=<-2 3 5>, 2=?, 3=?, Index=$Particles.PartitionParticles001.Partition, Num=::NewEnumerator0)
    Skeletal Mesh=$System.Skeletal Mesh
    Source Movement Scale:Vector3f=<0 0 0>
   )
   &ApplyInitialForces()
  ]
  particleUpdateScript [
   &ParticleState()
   &GravityForce(Gravity=<0 0 -150>) {enabled=false}
   &ScaleSpriteSize(Scale Factor=&Vector2DFromFloat(Value=&FloatFromCurve(
      FloatCurve=DataInterface() {propertyValues {Curve {defaultValue=0.5, keys [{interpMode=RCIM_Cubic, tangentMode=RCTM_User, time=0, value=0.75}, {tangentMode=RCTM_User, time=1, value=0.3}]}}}
      Scale Curve=1.5
     )))
   &Drag(Drag=2.5, Use Linear Drag=true, Use Rotational Drag=true)
   &InheritSourceMovement(ApplyForce=false, ApplyPosition=true, ApplyPositionFalloff=false, Source Velocity:Vector3f=$Emitter.InheritVelocityBlend)
   &InheritVelocity(Inherited Velocity Speed Limit=250, Source Speed Threshold=50, Speed Limit=true, UseSpeedThreshold=true)
   &PointAttractionForce(
    Attraction Radius=5
    Attraction Strength=1500
    Attractor Position:NiagaraPosition=$Particles.Initial.ShapeLocation.ShapePosition
    Kill Radius Overshoot Correction=5
    Kill Within Radius=true
    Module.Use Falloff=true
    Position To Attract:NiagaraPosition=$Particles.SkeletalMeshLocation.SampledPosition
    UseAttractorOffset=false
   )
   &SolveForcesAndVelocity(Clamp Velocity=false, Limit Acceleration=false)
   &SubUVAnimation(
    End Frame=25
    Sub UV Play Rate:NiagaraFloat=&Multiply_Float(A=0.1, B=$Particles.NormalizedAge)
    SubUV Animation Mode:ENiagara_SubUVLookupMode=::NewEnumerator3
   ) {module=SubUVAnimation001}
   &SpriteRotationRate(Rotation Rate=&FloatFromCurve(
     FloatCurve=DataInterface() {propertyValues {Curve {keys [<at 0 0 RCIM_Cubic>, <at 0.3 1 RCIM_Cubic>, {arriveTangent=-1.7543, tangentMode=RCTM_User, time=1, value=0}]}}}
     Scale Curve=&RandomRangeFloat(Minimum=-20, Maximum=56)
    ))
   &ScaleColor(
    Scale Alpha=&FloatFromCurve(
     FloatCurve=DataInterface() {propertyValues {Curve {
        defaultValue=1
        keys [
         {interpMode=RCIM_Cubic, tangentMode=RCTM_User, time=0, value=0}
         {interpMode=RCIM_Cubic, tangentMode=RCTM_User, time=0.0884, value=0.9718}
         {tangentMode=RCTM_User, time=0.7, value=0}
        ]
       }}}
     Scale Curve=&Multiply_Float(
      A=&FNTimeOfDayScalar(Night Scale=0.035, Morning Scale=0.03, Day Scale=0.035, Evening Scale=0.02)
      B=$User.FrostyMistAlpha
     )
    )
    ScaleA=true
    ScaleRGB=true
   ) {module=ScaleColor001}
  ]
  rendererClasses [&NiagaraSpriteRendererProperties]
  renderers [
   {rendererClass=&NiagaraSpriteRendererProperties}
   {rendererClass=&NiagaraSpriteRendererProperties, rendererIndex=1}
  ]
 }
 Frozen_mist_Burst {
  emitterUpdateScript [
   &EmitterState(Life Cycle Mode=::NewEnumerator1, Loop Behavior=::NewEnumerator1, Loop Delay=1.3, Loop Duration=0.5, UseLoopDelay=true)
   &SpawnBurst_Instantaneous(Spawn Count=30, Use Spawn Probability=false, UseLoopCountLimit=false)
  ]
  particleSpawnScript [
   &PartitionParticles(Number Of Partitions=5)
   &InitializeParticle(
    Color=$System.TOD Snowy Mist
    Lifetime Max=0.65
    Lifetime Min=0.35
    Lifetime Mode=::NewEnumerator1
    Sprite Rotation Mode=::NewEnumerator2
    Sprite Size Mode=::NewEnumerator1
    Uniform Sprite Size Max=100
    Uniform Sprite Size Min=75
    UsePositionOffset=false
    Write Material Random=true
   ) {module=InitializeParticle001}
   &SystemLocation(MaskSpawnGroup=false, Offset=<0 0 9>)
   &CameraOffset(Camera Offset Amount=&UniformRangedFloat(Minimum=4, Maximum=10))
   /SetVariables(Particles.SpriteFacing=<0.05 0.05 1>) {isSetParameters=true}
   &AddVelocity(Velocity=&RandomRangeVector(Minimum=<-110 -110 0>, Maximum=<110 110 1>)) {module=AddVelocity001}
  ]
  particleUpdateScript [
   &Drag(
    Drag=&FloatFromCurve(
     FloatCurve=DataInterface() {propertyValues {Curve {defaultValue=1, keys [{interpMode=RCIM_Cubic, tangentMode=RCTM_Break, time=0.4, value=17.5}, <at 0.4147 27.5>, <at 0.5149 60>]}}}
     Scale Curve=0.01
    )
    Use Linear Drag=true
    Use Rotational Drag=false
   )
   &GravityForce(Gravity=<0 0 0>)
   &SolveForcesAndVelocity(Clamp Velocity=false, Limit Acceleration=false)
   &ParticleState()
   &ScaleSpriteSize()
   &SpriteRotationRate(Rotation Rate=&RandomRangeFloat(Minimum=20, Maximum=160))
   &ScaleColor(
    Scale Alpha=&FloatFromCurve(
     FloatCurve=DataInterface() {propertyValues {Curve {keys [
         {arriveTangent=4.4407, interpMode=RCIM_Cubic, leaveTangent=4.4407, tangentMode=RCTM_User, time=0, value=0}
         {arriveTangent=0.0345, interpMode=RCIM_Cubic, leaveTangent=0.0345, tangentMode=RCTM_User, time=0.2332, value=0.3324}
         {arriveTangent=-0.9753, interpMode=RCIM_Cubic, leaveTangent=-0.9753, tangentMode=RCTM_User, time=0.85, value=0}
        ]}}}
     Scale Curve=&FNTimeOfDayScalar(Night Scale=0.06, Morning Scale=0.06, Day Scale=0.07, Evening Scale=0.03)
    )
    ScaleA=true
    ScaleRGB=false
   )
   &SubUVAnimation(
    "Sprite Renderer "=DataInterface() {propertyValues {SpriteRenderer=&self:Frozen_mist_Burst_0.NiagaraSpriteRendererProperties_2}}
    UseEndFrame=false
    UseStartFrame=false
   )
   /SetVariables(Particles.1 - Ground Fog Dyn Param:Vector4f=&MakeVector4(X=0, Y=1, Z=0.1, W=1)) {isSetParameters=true}
  ]
  rendererClasses [&NiagaraSpriteRendererProperties]
  renderers [{rendererClass=&NiagaraSpriteRendererProperties}]
 }
 Ground_burst {
  emitterSpawnScript [/SetVariables(Emitter.EmissiveAdjustment=1) {isSetParameters=true}]
  emitterUpdateScript [
   &EmitterState(Life Cycle Mode=::NewEnumerator1, Loop Behavior=::NewEnumerator1, Loop Delay=1.32, Loop Duration=1, UseLoopDelay=true)
   &SpawnBurst_Instantaneous(Spawn Count=30, Spawn Group=2, Use Spawn Probability=false, UseLoopCountLimit=false) {module=SpawnBurst_Instantaneous002}
   &SpawnBurst_Instantaneous(Spawn Count=25, Spawn Group=1, Use Spawn Probability=false, UseLoopCountLimit=false)
   &SpawnBurst_Instantaneous(Spawn Count=17, Use Spawn Probability=false, UseLoopCountLimit=false) {module=SpawnBurst_Instantaneous001}
  ]
  particleSpawnScript [
   /SetVariables(Particles.VisibilityTag=&SpawnGroup()) {isSetParameters=true, module=SetVariables_#4}
   &ShapeLocation(
    MaskSpawnGroup=false
    Ring / Disc Distribution Mode:ENiagaraSphereDistributionMode=::NewEnumerator3
    Ring Radius=50
    Shape Primitive=::NewEnumerator4
    Uniform Spiral Falloff=1.5
    UseOwnerScale=true
   )
   &InitializeParticle(
    Color=&MergeHelper_ColorSwitch(
     0=<rgba 0.6344 0.805 1 2>
     1=<rgba 0.6344 0.805 1 2>
     2=$System.TOD Snowy Mist
     3=?
     Index=&SpawnGroup()
     Num=::NewEnumerator1
    )
    Lifetime=&MergeHelper_FloatSwitch(
     0=&RandomRangeFloat(Minimum=0.2, Maximum=0.65)
     1=&RandomRangeFloat(Minimum=0.2, Maximum=0.65)
     2=&RandomRangeFloat(Minimum=0.35, Maximum=0.65)
     3=?
     Index=&SpawnGroup()
     Num=::NewEnumerator1
    )
    Position Mode=::NewEnumerator1
    Position Offset=<0 0 0.5>
    Sprite Rotation Mode=::NewEnumerator2
    Sprite Size Mode=::NewEnumerator3
    Uniform Sprite Size=&MergeHelper_FloatSwitch(
     0=/Niagara/DynamicInputs/Helpers/DistributionCurveFloat(
      Minimum:NiagaraFloat=5
      Maximum:NiagaraFloat=17
      Distribution Curve:NiagaraDataInterfaceCurve=DataInterface() {propertyValues {Curve {defaultValue=0, keys [<at 0 0 RCIM_Constant>,<at 0.15 0.5 RCIM_Constant>,<at 0.8 1 RCIM_Constant>]}}}
      Randomness Mode:ENiagaraRandomnessMode=::NewEnumerator0
      Evaluation Type:ENiagaraRandomnessEvaluation=::NewEnumerator0
      Seed:NiagaraInt32=0
      Fixed Random Seed:NiagaraBool=false
     )
     1=&RandomRangeFloat(Minimum=2, Maximum=8)
     2=&RandomRangeFloat(Minimum=75, Maximum=100)
     3=?
     Index=$Particles.VisibilityTag
     Num=::NewEnumerator1
    )
    UsePositionOffset=true
    Write Material Random=true
   )
   &AddVelocityFromPoint(
    Module.Use Falloff Distance=false
    Origin Offset=&MergeHelper_VectorSwitch(
     0=&MakeVector(
      X=&RandomRangeFloat(Minimum=-10, Maximum=10)
      Y=&RandomRangeFloat(Minimum=-10, Maximum=10)
      Z=-0.5
     )
     1=&MakeVector(
      X=&RandomRangeFloat(Minimum=-10, Maximum=10)
      Y=&RandomRangeFloat(Minimum=-10, Maximum=10)
      Z=-0.5
     )
     2=&MakeVector(
      X=&RandomRangeFloat(Minimum=-10, Maximum=10)
      Y=&RandomRangeFloat(Minimum=-10, Maximum=10)
      Z=-0.5
     )
     3=?
     Index=&SpawnGroup()
     Num=::NewEnumerator1
    )
    Velocity Strength=&MergeHelper_FloatSwitch(
     0=&Multiply_Float(A=&RandomRangeFloat(Minimum=0.5, Maximum=1), B=225)
     1=&Multiply_Float(A=&RandomRangeFloat(Minimum=0.5, Maximum=1), B=225)
     2=&Multiply_Float(A=&RandomRangeFloat(Minimum=0.8, Maximum=1.3), B=225)
     3=?
     Index=&SpawnGroup()
     Num=::NewEnumerator1
    )
   )
   &ApplyInitialForces()
   &DynamicMaterialParameters(
    Index 0 Param 1=&RandomRangeFloat(Minimum=0.3, Maximum=0)
    Index 0 Param 2=&RandomRangeFloat(Minimum=1, Maximum=2)
    Index 0 Param 3=&RandomRangeFloat(Minimum=-1, Maximum=1)
    Module.Index0Param0Write=true
    Module.Index0Param1Write=true
    Module.Index0Param2Write=true
    Module.Index0Param3Write=true
   )
   &CameraOffset(Camera Offset Amount=&MergeHelper_FloatSwitch(
     0=15
     1=15
     2=&UniformRangedFloat(Minimum=4, Maximum=10)
     3=?
     Index=&SpawnGroup()
     Num=::NewEnumerator1
    ))
  ]
  particleUpdateScript [
   &ParticleState()
   &SpriteRotationRate(Rotation Rate=&MergeHelper_FloatSwitch(
     0=&RandomRangeFloat(Minimum=75, Maximum=250)
     1=&RandomRangeFloat(Minimum=75, Maximum=250)
     2=&RandomRangeFloat(Minimum=20, Maximum=160)
     3=?
     Index=&SpawnGroup()
     Num=::NewEnumerator1
    ))
   &ScaleSpriteSize(
    Scale Sprite Size Mode=::NewEnumerator0
    Uniform Scale Factor=&MergeHelper_FloatSwitch(
     0=&FloatFromCurve(FloatCurve=DataInterface() {propertyValues {Curve {keys [
          {interpMode=RCIM_Cubic, tangentMode=RCTM_Break, time=0, value=0}
          {interpMode=RCIM_Cubic, tangentMode=RCTM_Break, time=0.2, value=0.85}
          {interpMode=RCIM_Cubic, tangentMode=RCTM_Break, time=0.5, value=0.95}
          {interpMode=RCIM_Cubic, tangentMode=RCTM_Break, time=1, value=0}
         ]}}})
     1=&FloatFromCurve(FloatCurve=DataInterface() {propertyValues {Curve {keys [
          {interpMode=RCIM_Cubic, tangentMode=RCTM_Break, time=0, value=0}
          {interpMode=RCIM_Cubic, tangentMode=RCTM_Break, time=0.2, value=0.85}
          {interpMode=RCIM_Cubic, tangentMode=RCTM_Break, time=0.5, value=0.95}
          {interpMode=RCIM_Cubic, tangentMode=RCTM_Break, time=1, value=0}
         ]}}})
     2=&FloatFromCurve(
      FloatCurve=DataInterface() {propertyValues {Curve {keys [{tangentMode=RCTM_User, time=0, value=0}, {tangentMode=RCTM_User, time=1, value=1}]}}}
      Scale Curve=1.5
     )
     3=?
     Index=&SpawnGroup()
     Num=::NewEnumerator1
    )
   )
   &Drag(
    Drag=&MergeHelper_FloatSwitch(
     0=&FloatFromCurve(
      FloatCurve=DataInterface() {propertyValues {Curve {defaultValue=1, keys [{interpMode=RCIM_Cubic, tangentMode=RCTM_Break, time=0.4, value=17.5}, <at 0.4149 17.5>, <at 0.5149 10>]}}}
      Scale Curve=0.2
     )
     1=&FloatFromCurve(
      FloatCurve=DataInterface() {propertyValues {Curve {defaultValue=1, keys [{interpMode=RCIM_Cubic, tangentMode=RCTM_Break, time=0.4, value=17.5}, <at 0.4149 17.5>, <at 0.5149 10>]}}}
      Scale Curve=0.2
     )
     2=&FloatFromCurve(
      FloatCurve=DataInterface() {propertyValues {Curve {defaultValue=1, keys [{interpMode=RCIM_Cubic, tangentMode=RCTM_Break, time=0.4, value=17.5}, <at 0.4147 27.5>, <at 0.5149 60>]}}}
      Scale Curve=0.25
     )
     3=?
     Index=&SpawnGroup()
     Num=::NewEnumerator1
    )
    Use Linear Drag=true
    Use Rotational Drag=true
   )
   &GravityForce(Gravity=&MergeHelper_VectorSwitch(
     0=&MakeVector(
      X=0
      Y=0
      Z=&FloatFromCurve(
       FloatCurve=DataInterface() {propertyValues {Curve {
          defaultValue=1
          keys [{interpMode=RCIM_Cubic, tangentMode=RCTM_User, time=0, value=0}, {interpMode=RCIM_Cubic, tangentMode=RCTM_User, time=1, value=0.5}]
         }}}
       Scale Curve=50
      )
     )
     1=&MakeVector(
      X=0
      Y=0
      Z=&FloatFromCurve(
       FloatCurve=DataInterface() {propertyValues {Curve {
          defaultValue=1
          keys [{interpMode=RCIM_Cubic, tangentMode=RCTM_User, time=0, value=0}, {interpMode=RCIM_Cubic, tangentMode=RCTM_User, time=1, value=0.5}]
         }}}
       Scale Curve=50
      )
     )
     2=<0 0 0>
     3=?
     Index=&SpawnGroup()
     Num=::NewEnumerator1
    ))
   &CurlNoiseForce(Noise Frequency=2, Noise Strength=50, Pan Noise=true, Pan Noise Field=<0 0 0.5>) {module=CurlNoiseForce001}
   &SolveForcesAndVelocity(Clamp Velocity=false, Limit Acceleration=false)
   /SetVariables(Particles.GlowScale:Vector2f=&Vector2DFromFloat(Value=&FloatFromCurve(
      FloatCurve=DataInterface() {propertyValues {Curve {
         defaultValue=0
         keys [
          {interpMode=RCIM_Cubic, tangentMode=RCTM_User, time=0, value=0}
          {arriveTangent=2, interpMode=RCIM_Cubic, leaveTangent=2, time=0.2204, value=0.8013}
          {interpMode=RCIM_Cubic, tangentMode=RCTM_User, time=0.5, value=1}
          {tangentMode=RCTM_User, time=1, value=0}
         ]
        }}}
      Scale Curve=8.5
     ))) {isSetParameters=true, module=SetVariables_#5}
   /SetVariables(
    Particles.SnowFlake Color:LinearColor=&ColorFromCurve(ColorCurve=DataInterface() {propertyValues {
       AlphaCurve {keys [<at 0.0926 0 RCIM_Cubic>,<at 0.5192 0.5>,<at 0.8563 0 RCIM_Cubic>]}
       BlueCurve {keys [<at 0 1>,<at 0.5264 1 RCIM_Cubic>,<at 1 1 RCIM_Cubic>]}
       GreenCurve {keys [<at 0 1>, {arriveTangent=-0.727, interpMode=RCIM_Cubic, leaveTangent=-0.727, time=0.5264, value=0.7228}, <at 1 0.273 RCIM_Cubic>]}
       RedCurve {keys [<at 0 1>, {arriveTangent=-0.8906, interpMode=RCIM_Cubic, leaveTangent=-0.8906, time=0.5264, value=0.4802}, <at 1 0.1094 RCIM_Cubic>]}
      }})
    Particles.Snowflake Size:Vector2f=&Multiply_Vector2D(
     A=&Vector2DFromFloat(Value=&FloatFromCurve(
       FloatCurve=DataInterface() {propertyValues {Curve {
          defaultValue=0
          keys [
           {interpMode=RCIM_Cubic, tangentMode=RCTM_User, time=0, value=0}
           {arriveTangent=2, interpMode=RCIM_Cubic, leaveTangent=2, time=0.1572, value=0.8324}
           {interpMode=RCIM_Cubic, tangentMode=RCTM_User, time=0.5, value=1}
           {tangentMode=RCTM_User, time=1, value=0}
          ]
         }}}
       Scale Curve=0.5
      ))
     B=$Particles.SpriteSize
    )
   ) {isSetParameters=true, module=SetVariables_#6}
   /SetVariables(
    Particles.Sparkle Color:LinearColor=&MakeLinearColorFromVectorAndFloat(
     "Vector (RGB)"=&Multiply_Vector(
      A=&RandomRangeVector(
       Minimum=&MakeVectorFromLinearColorRGB(Color=&ColorFromCurve(ColorCurve=DataInterface() {propertyValues {
           AlphaCurve {keys [<at 0.4411 1>]}
           BlueCurve {keys [<at 0 1>,<at 1 1 RCIM_Cubic>]}
           GreenCurve {keys [<at 0 0.8383>,<at 1 0.8402 RCIM_Cubic>]}
           RedCurve {keys [<at 0 0.6969>,<at 1 0.5991 RCIM_Cubic>]}
          }}))
       Maximum=&MakeVectorFromLinearColorRGB(Color=&ColorFromCurve(ColorCurve=DataInterface() {propertyValues {
           AlphaCurve {keys [<at 0.4411 1>]}
           BlueCurve {keys [<at 0 1 RCIM_Cubic>,<at 1 1 RCIM_Cubic>]}
           GreenCurve {keys [<at 0 0.7871 RCIM_Cubic>,<at 1 0.273 RCIM_Cubic>]}
           RedCurve {keys [<at 0 0.1635 RCIM_Cubic>,<at 1 0.1094 RCIM_Cubic>]}
          }}))
      )
      B=<3 3 3>
     )
     "Float (Alpha)"=&FloatFromCurve(
      FloatCurve=DataInterface() {propertyValues {Curve {
         defaultValue=1
         keys [
          {arriveTangent=9.0725, interpMode=RCIM_Cubic, leaveTangent=9.0725, tangentMode=RCTM_User, time=0, value=0}
          {arriveTangent=0.2227, interpMode=RCIM_Cubic, leaveTangent=0.2227, tangentMode=RCTM_User, time=0.1408, value=0.9847}
          {arriveTangent=-1.8564, interpMode=RCIM_Cubic, leaveTangent=-1.8564, tangentMode=RCTM_User, time=1, value=0}
         ]
        }}}
      Scale Curve=&FNTimeOfDayScalar(Night Scale=1, Morning Scale=1, Day Scale=5, Evening Scale=4.5)
     )
    )
    Particles.Sparkle DynParam:Vector4f=&MakeVector4(X=1, Y=1, Z=1, W=1)
    Particles.Sparkle Sprite Size:Vector2f=&Multiply_Vector2D(A=&RandomRangeVector2D(Minimum=<0.1 0.1>, Maximum=<0.75 0.75>), B=$Particles.SpriteSize)
   ) {isSetParameters=true, module=SetVariables_#7}
   /SetVariables(
    Particles.Snowflake Glow:LinearColor=&MakeLinearColorFromVectorAndFloat(
     "Vector (RGB)"=&Multiply_Vector(
      A=&RandomRangeVector(
       Minimum=&MakeVectorFromLinearColorRGB(Color=&ColorFromCurve(ColorCurve=DataInterface() {propertyValues {
           AlphaCurve {keys [<at 0 0 RCIM_Cubic>,<at 0.0335 0.5>,<at 1 0 RCIM_Cubic>]}
           BlueCurve {keys [<at 0 1 RCIM_Cubic>,<at 1 1 RCIM_Cubic>]}
           GreenCurve {keys [<at 0 0.4764 RCIM_Cubic>,<at 1 0.1837 RCIM_Cubic>]}
           RedCurve {keys [<at 0 0.0859 RCIM_Cubic>,<at 1 0 RCIM_Cubic>]}
          }}))
       Maximum=&MakeVectorFromLinearColorRGB(Color=&ColorFromCurve(ColorCurve=DataInterface() {propertyValues {
           AlphaCurve {keys [<at 0 0>,<at 0.1429 0.8>,<at 1 0>]}
           BlueCurve {keys [<at 0 0.7231 RCIM_Cubic>,<at 1 1 RCIM_Cubic>]}
           GreenCurve {keys [<at 0 0.7406 RCIM_Cubic>,<at 1 0.273 RCIM_Cubic>]}
           RedCurve {keys [<at 0 0.1921 RCIM_Cubic>,<at 1 0.1094 RCIM_Cubic>]}
          }}))
      )
      B=<2 2 2>
     )
     "Float (Alpha)"=&FloatFromCurve(
      FloatCurve=DataInterface() {propertyValues {Curve {
         defaultValue=1
         keys [
          {arriveTangent=9.0725, interpMode=RCIM_Cubic, leaveTangent=9.0725, tangentMode=RCTM_User, time=0, value=0}
          {arriveTangent=0.2227, interpMode=RCIM_Cubic, leaveTangent=0.2227, tangentMode=RCTM_User, time=0.1408, value=0.9847}
          {arriveTangent=-1.8564, interpMode=RCIM_Cubic, leaveTangent=-1.8564, tangentMode=RCTM_User, time=1, value=0}
         ]
        }}}
      Scale Curve=&FNTimeOfDayScalar(Night Scale=0.5, Morning Scale=0.5, Evening Scale=0.4)
     )
    )
    Particles.Snowflake Glow Size:Vector2f=&Multiply_Vector2D(A=<0.65 0.65>, B=$Particles.SpriteSize)
    Particles.Snowflake glow dynparm:Vector4f=&MakeVector4(
     X=&FloatFromCurve(FloatCurve=DataInterface() {propertyValues {Curve {keys [<at 0 1.1>,<at 0.2 0>,<at 0.85 0>,<at 1 1>]}}})
     Y=98.7492
     Z=0
     W=0
    )
   ) {isSetParameters=true, module=SetVariables_#8}
   /SetVariables(
    Particles.Frost mist DynParam 1:Vector4f=&MakeVector4(X=0, Y=1, Z=0.1, W=1)
    Particles.Frosty Mist SpriteSize:Vector2f=&Multiply_Vector2D_ByFloat(Float=50, Vector2D=<1.2 1.2>)
    Particles.Frosty Mist color:LinearColor=&MakeLinearColorFromVectorAndFloat(
     "Vector (RGB)"=&MakeVectorFromLinearColorRGB(Color=<rgba 0.5385 0.7537 1 1>)
     "Float (Alpha)"=&FloatFromCurve(
      FloatCurve=DataInterface() {propertyValues {Curve {keys [
          {arriveTangent=4.4407, interpMode=RCIM_Cubic, leaveTangent=4.4407, tangentMode=RCTM_User, time=0, value=0}
          {arriveTangent=0.0345, interpMode=RCIM_Cubic, leaveTangent=0.0345, tangentMode=RCTM_User, time=0.2332, value=0.3324}
          {arriveTangent=-1.3404, interpMode=RCIM_Cubic, leaveTangent=-1.3404, tangentMode=RCTM_User, time=0.85, value=0}
         ]}}}
      Scale Curve=&FNTimeOfDayScalar(Night Scale=0.075, Morning Scale=0.075, Day Scale=0.085, Evening Scale=0.045)
     )
    )
    Particles.Frosty Mist velocity:Vector3f=&RandomRangeVector(Minimum=<-100 -100 0>, Maximum=<100 100 1>)
   ) {isSetParameters=true}
   &SubUVAnimation(
    "Sprite Renderer "=DataInterface() {propertyValues {SpriteRenderer=&self:ground_burst_1.NiagaraSpriteRendererProperties_4}}
    UseEndFrame=false
    UseStartFrame=false
   )
  ]
  rendererClasses [&NiagaraSpriteRendererProperties]
  renderers [
   {rendererClass=&NiagaraSpriteRendererProperties}
   {rendererClass=&NiagaraSpriteRendererProperties, rendererIndex=1}
   {rendererClass=&NiagaraSpriteRendererProperties, rendererIndex=2}
   {rendererClass=&NiagaraSpriteRendererProperties, rendererIndex=3}
   {rendererClass=&NiagaraSpriteRendererProperties, rendererIndex=4}
  ]
 }
 Hand_Sparkles {
  emitterUpdateScript [
   &EmitterState()
   &SpawnRate(SpawnGroup=1, SpawnRate=40, Use Spawn Probability=false) {module=SpawnRate001}
   &SpawnRate(SpawnRate=15, Use Spawn Probability=false)
  ]
  particleSpawnScript [
   /SetVariables(Particles.VisibilityTag=&SpawnGroup()) {isSetParameters=true, module=SetVariables_#4}
   &PartitionParticles() {module=PartitionParticles001}
   &InitializeParticle(
    Color=<rgba 0.6385 0.8072 1 2>
    Lifetime Max=0.7
    Lifetime Min=0.35
    Lifetime Mode=::NewEnumerator1
    Sprite Rotation Mode=::NewEnumerator2
    Sprite Size Mode=::NewEnumerator3
    Uniform Sprite Size=&MergeHelper_FloatSwitch(
     0=&DistributionCurveFloat(
      Minimum:NiagaraFloat=5
      Maximum:NiagaraFloat=22
      Distribution Curve:NiagaraDataInterfaceCurve=DataInterface() {propertyValues {Curve {defaultValue=0, keys [<at 0 0 RCIM_Constant>,<at 0.15 0.5 RCIM_Constant>,<at 0.8 1 RCIM_Constant>]}}}
      Randomness Mode:ENiagaraRandomnessMode=::NewEnumerator0
      Evaluation Type:ENiagaraRandomnessEvaluation=::NewEnumerator0
      Seed:NiagaraInt32=0
      Fixed Random Seed:NiagaraBool=false
     )
     1=&RandomRangeFloat(Minimum=3, Maximum=7)
     2=?
     3=?
     Index=$Particles.VisibilityTag
     Num=::NewEnumerator0
    )
    UsePositionOffset=false
    Write Material Random=true
   )
   &SkeletalMeshLocation(
    Bone / Socket Index=$Particles.PartitionParticles001.Partition
    Bone Sampling Mode=::NewEnumerator4
    MaskSpawnGroup=false
    Skeletal Mesh=$System.Skeletal Mesh
   )
   &ShapeLocation(MaskSpawnGroup=false, Override Seed=false, Sphere Radius=10, UseOwnerScale=true)
   &InheritVelocity(Speed Limit=true, UseSpeedThreshold=false)
   &DynamicMaterialParameters(
    Index 0 Param 1=&RandomRangeFloat(Minimum=0.3, Maximum=0)
    Index 0 Param 2=&RandomRangeFloat(Minimum=1, Maximum=2)
    Index 0 Param 3=&RandomRangeFloat(Minimum=-1, Maximum=1)
    Module.Index0Param0Write=true
    Module.Index0Param1Write=true
    Module.Index0Param2Write=true
    Module.Index0Param3Write=true
   )
  ]
  particleUpdateScript [
   &ParticleState()
   &ScaleSpriteSize(Uniform Curve Sprite Scale=DataInterface() {propertyValues {Curve {keys [{tangentMode=RCTM_User, time=0, value=0.05}, <at 0.2 0.85>, <at 0.5 0.95>, {tangentMode=RCTM_User, time=1, value=0}]}}})
   &GravityForce(Gravity=<0 0 -200>)
   &SpriteRotationRate(Rotation Rate=&RandomRangeFloat(Minimum=80, Maximum=200, Evaluation Type=::NewEnumerator1))
   &Drag(Drag=3, Use Linear Drag=true, Use Rotational Drag=true)
   &SolveForcesAndVelocity(Clamp Velocity=false, Limit Acceleration=false)
   /SetVariables(Particles.GlowScale:Vector2f=&Vector2DFromFloat(Value=&FloatFromCurve(
      FloatCurve=DataInterface() {propertyValues {Curve {
         defaultValue=0
         keys [
          {interpMode=RCIM_Cubic, tangentMode=RCTM_User, time=0, value=0}
          {arriveTangent=2, interpMode=RCIM_Cubic, leaveTangent=2, time=0.2204, value=0.8013}
          {interpMode=RCIM_Cubic, tangentMode=RCTM_User, time=0.5, value=1}
          {tangentMode=RCTM_User, time=1, value=0}
         ]
        }}}
      Scale Curve=7
     ))) {isSetParameters=true, module=SetVariables_#5}
   /SetVariables(Particles.Line Color:LinearColor=&MultiplyLinearColor(
     Linear Color=<rgba 3.8016 8.944 10 1>
     Scale Factor=&MakeVector4(
      X=1
      Y=1
      Z=1.5
      W=&Multiply_Float(
       A=&Lerp_Float(
        A=0.005
        B=0.01
        Alpha=&FloatFromCurve(FloatCurve=DataInterface() {propertyValues {Curve {keys [<at 0 0>,<at 0.096 0.8735>,<at 0.1302 0.0844>,<at 0.2172 0.8972>,<at 0.2722 -0.0207>,<at 0.3649 0.909>,<at 0.4194 0.0849>,<at 0.5166 0.9082>,<at 0.5313 0.0253>,<at 0.6308 0.9318>,<at 0.7036 0.006>,<at 0.7872 0.8487>,<at 0.849 0.0772>,<at 0.9054 0.8724>,<at 1 1>]}}})
       )
       B=&RandomRangeFloat(Minimum=0.05, Maximum=0.1)
      )
     )
    )) {isSetParameters=true}
   /SetVariables(Particles.LineScale:Vector2f=&Multiply_Vector2D(A=&RandomRangeVector2D(Minimum=<1 1>, Maximum=<4 4>), B=$Particles.SpriteSize)) {isSetParameters=true}
   /SetVariables(
    Particles.SnowFlake Color:LinearColor=&ColorFromCurve(ColorCurve=DataInterface() {propertyValues {
       AlphaCurve {keys [<at 0.0926 0 RCIM_Cubic>,<at 0.2667 0.25>,<at 0.7202 0 RCIM_Cubic>]}
       BlueCurve {keys [<at 0 1 RCIM_Cubic>,<at 1 1 RCIM_Cubic>]}
       GreenCurve {keys [<at 0 0.7228 RCIM_Cubic>,<at 1 0.6389 RCIM_Cubic>]}
       RedCurve {keys [<at 0 0.4802 RCIM_Cubic>,<at 1 0.5078 RCIM_Cubic>]}
      }})
    Particles.Snowflake Size:Vector2f=&Vector2DFromFloat(Value=&FloatFromCurve(
      FloatCurve=DataInterface() {propertyValues {Curve {
         defaultValue=0
         keys [
          {interpMode=RCIM_Cubic, tangentMode=RCTM_User, time=0, value=0}
          {arriveTangent=2, interpMode=RCIM_Cubic, leaveTangent=2, time=0.2204, value=0.8013}
          {interpMode=RCIM_Cubic, tangentMode=RCTM_User, time=0.5, value=1}
          {tangentMode=RCTM_User, time=1, value=0}
         ]
        }}}
      Scale Curve=6
     ))
   ) {isSetParameters=true, module=SetVariables_#6}
   /SetVariables(
    Particles.Sparkle Color:LinearColor=&MakeLinearColorFromVectorAndFloat(
     "Vector (RGB)"=&Multiply_Vector(
      A=&RandomRangeVector(
       Minimum=&MakeVectorFromLinearColorRGB(Color=&ColorFromCurve(ColorCurve=DataInterface() {propertyValues {
           AlphaCurve {keys [<at 0.4411 1>]}
           BlueCurve {keys [<at 0 1>,<at 1 1 RCIM_Cubic>]}
           GreenCurve {keys [<at 0 0.8383>,<at 1 0.8402 RCIM_Cubic>]}
           RedCurve {keys [<at 0 0.6969>,<at 1 0.5991 RCIM_Cubic>]}
          }}))
       Maximum=&MakeVectorFromLinearColorRGB(Color=&ColorFromCurve(ColorCurve=DataInterface() {propertyValues {
           AlphaCurve {keys [<at 0.4411 1>]}
           BlueCurve {keys [<at 0 1 RCIM_Cubic>,<at 1 1 RCIM_Cubic>]}
           GreenCurve {keys [<at 0 0.7628 RCIM_Cubic>,<at 1 0.273 RCIM_Cubic>]}
           RedCurve {keys [<at 0 0.0681 RCIM_Cubic>,<at 1 0.1094 RCIM_Cubic>]}
          }}))
      )
      B=<3 3 3>
     )
     "Float (Alpha)"=&FloatFromCurve(
      FloatCurve=DataInterface() {propertyValues {Curve {
         defaultValue=1
         keys [
          {arriveTangent=9.0725, interpMode=RCIM_Cubic, leaveTangent=9.0725, tangentMode=RCTM_User, time=0, value=0}
          {arriveTangent=0.2227, interpMode=RCIM_Cubic, leaveTangent=0.2227, tangentMode=RCTM_User, time=0.1408, value=0.9847}
          {arriveTangent=-1.8564, interpMode=RCIM_Cubic, leaveTangent=-1.8564, tangentMode=RCTM_User, time=1, value=0}
         ]
        }}}
      Scale Curve=&FNTimeOfDayScalar(Night Scale=3.5, Morning Scale=5, Day Scale=5, Evening Scale=3.5)
     )
    )
    Particles.Sparkle DynParams:Vector4f=&MakeVector4(X=0.2, Y=1, Z=1, W=1)
    Particles.Sparkle Sprite Size:Vector2f=&Multiply_Vector2D(A=&RandomRangeVector2D(Minimum=<0.2 0.2>, Maximum=<0.6 0.6>), B=$Particles.SpriteSize)
   ) {isSetParameters=true, module=SetVariables_#7}
   /SetVariables(Particles.Snowflake Glow:LinearColor=&MakeLinearColorFromVectorAndFloat(
     "Vector (RGB)"=&Multiply_Vector(
      A=&RandomRangeVector(
       Minimum=&MakeVectorFromLinearColorRGB(Color=&ColorFromCurve(ColorCurve=DataInterface() {propertyValues {
           AlphaCurve {keys [<at 0 0 RCIM_Cubic>,<at 0.0335 0.5>,<at 1 0 RCIM_Cubic>]}
           BlueCurve {keys [<at 0 1 RCIM_Cubic>,<at 1 1 RCIM_Cubic>]}
           GreenCurve {keys [<at 0 0.4764 RCIM_Cubic>,<at 1 0.1837 RCIM_Cubic>]}
           RedCurve {keys [<at 0 0.0859 RCIM_Cubic>,<at 1 0 RCIM_Cubic>]}
          }}))
       Maximum=&MakeVectorFromLinearColorRGB(Color=&ColorFromCurve(ColorCurve=DataInterface() {propertyValues {
           AlphaCurve {keys [<at 0 0>,<at 0.1429 0.25>,<at 1 0>]}
           BlueCurve {keys [<at 0 0.7231 RCIM_Cubic>,<at 1 1 RCIM_Cubic>]}
           GreenCurve {keys [<at 0 0.7406 RCIM_Cubic>,<at 1 0.273 RCIM_Cubic>]}
           RedCurve {keys [<at 0 0.1921 RCIM_Cubic>,<at 1 0.1094 RCIM_Cubic>]}
          }}))
      )
      B=<1 1 1>
     )
     "Float (Alpha)"=&FloatFromCurve(
      FloatCurve=DataInterface() {propertyValues {Curve {
         defaultValue=1
         keys [
          {arriveTangent=9.0725, interpMode=RCIM_Cubic, leaveTangent=9.0725, tangentMode=RCTM_User, time=0, value=0}
          {arriveTangent=0.2227, interpMode=RCIM_Cubic, leaveTangent=0.2227, tangentMode=RCTM_User, time=0.1408, value=0.9847}
          {arriveTangent=-1.8564, interpMode=RCIM_Cubic, leaveTangent=-1.8564, tangentMode=RCTM_User, time=1, value=0}
         ]
        }}}
      Scale Curve=&FNTimeOfDayScalar(Night Scale=0.27, Morning Scale=0.25, Day Scale=0.25, Evening Scale=0.25)
     )
    )) {isSetParameters=true, module=SetVariables_#8}
  ]
  rendererClasses [&NiagaraSpriteRendererProperties]
  renderers [
   {rendererClass=&NiagaraSpriteRendererProperties}
   {rendererClass=&NiagaraSpriteRendererProperties, rendererIndex=1}
   {rendererClass=&NiagaraSpriteRendererProperties, rendererIndex=2}
  ]
 }
 Hand_flake_off {
  emitterUpdateScript [
   &EmitterState(Life Cycle Mode=::NewEnumerator1, Loop Behavior=::NewEnumerator1, Loop Delay=0.3, Loop Duration=0.5, UseLoopDelay=true)
   &SpawnBurst_Instantaneous(Spawn Count=7, Spawn Group=1, Use Spawn Probability=false, UseLoopCountLimit=false) {module=SpawnBurst_Instantaneous001}
   &SpawnRate(SpawnRate=15, Use Spawn Probability=false) {enabled=false}
   &SpawnBurst_Instantaneous(Spawn Count=7, Use Spawn Probability=false, UseLoopCountLimit=false)
  ]
  particleSpawnScript [
   /SetVariables(Particles.VisibilityTag=&SpawnGroup()) {isSetParameters=true}
   &InitializeParticle(
    Color=<rgba 0.6385 0.8072 1 2>
    Lifetime Max=0.5
    Lifetime Min=0.2
    Lifetime Mode=::NewEnumerator1
    Sprite Rotation Mode=::NewEnumerator2
    Sprite Size Mode=::NewEnumerator1
    Uniform Sprite Size Max=8
    Uniform Sprite Size Min=3.5
    UsePositionOffset=false
    Write Material Random=true
   )
   &SkeletalMeshLocation(
    Bone / Socket Index=1
    Bone Sampling Mode=::NewEnumerator4
    MaskSpawnGroup=false
    Sampled Position Offset=<-10 0 7>
    Skeletal Mesh=$System.Skeletal Mesh
   )
   &DynamicMaterialParameters(
    Index 0 Param 1=&RandomRangeFloat(Minimum=0.3, Maximum=0)
    Index 0 Param 2=&RandomRangeFloat(Minimum=1, Maximum=2)
    Index 0 Param 3=&RandomRangeFloat(Minimum=-1, Maximum=1)
    Module.Index0Param0Write=true
    Module.Index0Param1Write=true
    Module.Index0Param2Write=true
    Module.Index0Param3Write=true
   )
   &AccelerationForce(Acceleration=<-5 0 0>)
   &AddVelocity(
    Cone Angle=35
    Cone Axis=<-1 0 0>
    Module.UseConeFalloff=false
    Module.UseVelDistribution=false
    Override Seed=false
    Velocity Mode=::NewEnumerator2
    Velocity Speed=200
   )
   &ApplyInitialForces()
  ]
  particleUpdateScript [
   &ParticleState()
   &SpriteRotationRate(Rotation Rate=&RandomRangeFloat(Minimum=70, Maximum=250, Evaluation Type=::NewEnumerator1))
   &ScaleSpriteSize(Uniform Curve Sprite Scale=DataInterface() {propertyValues {Curve {keys [{tangentMode=RCTM_User, time=0, value=0.05}, <at 0.2 0.85>, <at 0.5 0.95>, {tangentMode=RCTM_User, time=1, value=0}]}}})
   &Drag(Drag=&RandomRangeFloat(Minimum=0, Maximum=8.5), Use Linear Drag=true, Use Rotational Drag=true)
   &GravityForce(Gravity=&MakeVector(X=-15, Y=0, Z=&RandomRangeFloat(Minimum=-50, Maximum=-150)))
   &SolveForcesAndVelocity(Clamp Velocity=false, Limit Acceleration=false)
   /SetVariables(
    Particles.SnowFlake Color:LinearColor=&ColorFromCurve(ColorCurve=DataInterface() {propertyValues {
       AlphaCurve {keys [<at 0.0926 0 RCIM_Cubic>,<at 0.2667 0.25>,<at 0.7202 0 RCIM_Cubic>]}
       BlueCurve {keys [<at 0 1 RCIM_Cubic>,<at 1 1 RCIM_Cubic>]}
       GreenCurve {keys [<at 0 0.7228 RCIM_Cubic>,<at 1 0.6389 RCIM_Cubic>]}
       RedCurve {keys [<at 0 0.4802 RCIM_Cubic>,<at 1 0.5078 RCIM_Cubic>]}
      }})
    Particles.Snowflake Size:Vector2f=&Vector2DFromFloat(Value=&Multiply_Float(
      A=&FloatFromCurve(FloatCurve=DataInterface() {propertyValues {Curve {
          defaultValue=0
          keys [
           {interpMode=RCIM_Cubic, tangentMode=RCTM_User, time=0, value=0}
           {arriveTangent=2, interpMode=RCIM_Cubic, leaveTangent=2, time=0.2204, value=0.8013}
           {interpMode=RCIM_Cubic, tangentMode=RCTM_User, time=0.5, value=1}
           {tangentMode=RCTM_User, time=1, value=0}
          ]
         }}})
      B=&RandomRangeFloat(Minimum=5, Maximum=9)
     ))
   ) {isSetParameters=true}
   /SetVariables(
    Particles.Sparkle Color:LinearColor=&MakeLinearColorFromVectorAndFloat(
     "Vector (RGB)"=&MakeVectorFromLinearColorRGB(Color=&MakeLinearColorFromVectorAndFloat(
       "Vector (RGB)"=&MakeVectorFromLinearColorRGB(Color=&ColorFromCurve(ColorCurve=DataInterface() {propertyValues {
           AlphaCurve {keys [<at 0.4411 1>]}
           BlueCurve {keys [<at 0 0.9774 RCIM_Cubic>,<at 0.7707 0.6569>,<at 1 1 RCIM_Cubic>]}
           GreenCurve {keys [<at 0 1 RCIM_Cubic>,<at 0.7707 0.3997>,<at 1 0.6389 RCIM_Cubic>]}
           RedCurve {keys [<at 0 0.2927 RCIM_Cubic>,<at 0.7707 0.8066>,<at 1 0.5078 RCIM_Cubic>]}
          }}))
       "Float (Alpha)"=2
      ))
     "Float (Alpha)"=0.5
    )
    Particles.Sparkle Sprite Size:Vector2f=&Multiply_Vector2D(A=&RandomRangeVector2D(Minimum=<0.5 0.5>, Maximum=<1 1>), B=$Particles.SpriteSize)
   ) {isSetParameters=true}
   /SetVariables(
    Particles.Snow Flake Glow Size:Vector2f=&Vector2DFromFloat(Value=&FloatFromCurve(
      FloatCurve=DataInterface() {propertyValues {Curve {
         defaultValue=0
         keys [
          {interpMode=RCIM_Cubic, tangentMode=RCTM_User, time=0, value=0}
          {arriveTangent=2, interpMode=RCIM_Cubic, leaveTangent=2, time=0.2204, value=0.8013}
          {interpMode=RCIM_Cubic, tangentMode=RCTM_User, time=0.5, value=1}
          {tangentMode=RCTM_User, time=1, value=0}
         ]
        }}}
      Scale Curve=7
     ))
    Particles.Snowflake Glow:LinearColor=&MakeLinearColorFromVectorAndFloat(
     "Vector (RGB)"=&Multiply_Vector(
      A=&RandomRangeVector(
       Minimum=&MakeVectorFromLinearColorRGB(Color=&ColorFromCurve(ColorCurve=DataInterface() {propertyValues {
           AlphaCurve {keys [<at 0.4411 1>]}
           BlueCurve {keys [<at 0 0.6569>,<at 1 0.6877 RCIM_Cubic>]}
           GreenCurve {keys [<at 0 0.3997>,<at 1 0.1176 RCIM_Cubic>]}
           RedCurve {keys [<at 0 0.8066>,<at 1 1 RCIM_Cubic>]}
          }}))
       Maximum=<1 1 1>
      )
      B=&MakeVectorFromLinearColorRGB(Color=&ColorFromCurve(ColorCurve=DataInterface() {propertyValues {
          AlphaCurve {keys [<at 0.4411 1>]}
          BlueCurve {keys [<at 0 0.9826 RCIM_Cubic>,<at 1 1 RCIM_Cubic>]}
          GreenCurve {keys [<at 0 1 RCIM_Cubic>,<at 1 0.273 RCIM_Cubic>]}
          RedCurve {keys [<at 0 0.4552 RCIM_Cubic>,<at 1 0.1094 RCIM_Cubic>]}
         }}))
     )
     "Float (Alpha)"=&FloatFromCurve(
      FloatCurve=DataInterface() {propertyValues {Curve {
         defaultValue=1
         keys [
          {arriveTangent=9.0725, interpMode=RCIM_Cubic, leaveTangent=9.0725, tangentMode=RCTM_User, time=0, value=0}
          {arriveTangent=0.2227, interpMode=RCIM_Cubic, leaveTangent=0.2227, tangentMode=RCTM_User, time=0.1408, value=0.9847}
          {arriveTangent=-1.8564, interpMode=RCIM_Cubic, leaveTangent=-1.8564, tangentMode=RCTM_User, time=1, value=0}
         ]
        }}}
      Scale Curve=&FNTimeOfDayScalar(Night Scale=0.7, Morning Scale=0.25, Day Scale=0.3, Evening Scale=0.25)
     )
    )
   ) {isSetParameters=true}
   /SetVariables(Particles.GlowColor=&MakeLinearColorFromVectorAndFloat(
     "Vector (RGB)"=&Multiply_Vector(
      A=&MakeVectorFromLinearColorRGB(Color=&ColorFromCurve(ColorCurve=DataInterface() {propertyValues {
          AlphaCurve {keys [<at 0 0 RCIM_Cubic>,<at 0.0335 0.25>,<at 1 0 RCIM_Cubic>]}
          BlueCurve {keys [<at 0 1 RCIM_Cubic>,<at 1 1 RCIM_Cubic>]}
          GreenCurve {keys [<at 0 0.676 RCIM_Cubic>,<at 1 0.1837 RCIM_Cubic>]}
          RedCurve {keys [<at 0 0.4344 RCIM_Cubic>,<at 1 0 RCIM_Cubic>]}
         }}))
      B=<1 1 1>
     )
     "Float (Alpha)"=&FloatFromCurve(
      FloatCurve=DataInterface() {propertyValues {Curve {
         defaultValue=1
         keys [
          {arriveTangent=9.0725, interpMode=RCIM_Cubic, leaveTangent=9.0725, tangentMode=RCTM_User, time=0, value=0}
          {arriveTangent=0.2227, interpMode=RCIM_Cubic, leaveTangent=0.2227, tangentMode=RCTM_User, time=0.1408, value=0.9847}
          {arriveTangent=-1.8564, interpMode=RCIM_Cubic, leaveTangent=-1.8564, tangentMode=RCTM_User, time=1, value=0}
         ]
        }}}
      Scale Curve=&FNTimeOfDayScalar(Night Scale=0.35, Morning Scale=0.25, Day Scale=0.3, Evening Scale=0.25)
     )
    )) {isSetParameters=true}
  ]
  rendererClasses [&NiagaraSpriteRendererProperties]
  renderers [
   {rendererClass=&NiagaraSpriteRendererProperties}
   {rendererClass=&NiagaraSpriteRendererProperties, rendererIndex=1}
   {rendererClass=&NiagaraSpriteRendererProperties, rendererIndex=2}
  ]
 }
 hand_trails {
  emitterUpdateScript [
   &EmitterState()
   &FN_SpawnAnimTrail(Spawn Mode:ENiagaraSkeletalSpawning=::NewEnumerator0)
  ]
  particleSpawnScript [
   &FN_InitializeAnimTrailBoneIndecies()
   &FN_InitializeAnimTrail(
    AnimTrailWidth=0.15
    Lifetime=&Multiply_Float(
     A=&FloatFromCurve(
      FloatCurve=DataInterface() {propertyValues {Curve {
         defaultValue=1
         keys [{interpMode=RCIM_Cubic, tangentMode=RCTM_User, time=0, value=1}, {interpMode=RCIM_Cubic, tangentMode=RCTM_User, time=0.4905, value=-0.0601}]
        }}}
      Scale Curve=0.9
     )
     B=$User.AnimTrail_SpawnRate
    )
    SkeletalMesh=$System.Skeletal Mesh
   )
   &PartitionParticles(Write Partition to Ribbon ID=::NewEnumerator0)
   &ApplyInitialForces()
   &SkeletalMeshLocation(
    Bone / Socket Index=$Particles.PartitionParticles.Partition
    Bone Sampling Mode=::NewEnumerator4
    MaskSpawnGroup=false
    Skeletal Mesh=$System.Skeletal Mesh
   )
  ]
  particleUpdateScript [
   &ParticleState()
   &FN_UpdateAnimTrail(SkeletalMesh=$System.Skeletal Mesh)
   &ScaleColor(
    Scale Alpha=&FloatFromCurve(
     FloatCurve=DataInterface() {propertyValues {Curve {keys [{arriveTangent=-3.0806, interpMode=RCIM_Cubic, leaveTangent=-3.0806, tangentMode=RCTM_User, time=0, value=1}, <at 0.85 0.2071 RCIM_Cubic>]}}}
     Scale Curve=0.5
    )
    Scale RGB=&MakeVectorFromLinearColorRGB(Color=&ColorFromCurve(ColorCurve=DataInterface() {propertyValues {
        AlphaCurve {keys [<at 0 0.8 RCIM_Cubic>,<at 0.4686 0 RCIM_Cubic>]}
        BlueCurve {keys [<at 0 0.2 RCIM_Cubic>]}
        GreenCurve {keys [<at 0 0.1694 RCIM_Cubic>]}
        RedCurve {keys [<at 0 0.1319 RCIM_Cubic>]}
       }}))
    ScaleA=true
    ScaleRGB=true
   ) {module=ScaleColor003}
   &ScaleRibbonWidth(Ribbon Width Scale=&FloatFromCurve(FloatCurve=DataInterface() {propertyValues {Curve {
        defaultValue=1
        keys [
         {arriveTangent=0.0733, interpMode=RCIM_Cubic, leaveTangent=0.0733, tangentMode=RCTM_User, time=0, value=1}
         {arriveTangent=-0.0405, interpMode=RCIM_Cubic, leaveTangent=-0.0405, tangentMode=RCTM_User, time=0.7, value=0.3}
        ]
       }}}))
   &InheritSourceMovement(ApplyForce=false, ApplyPosition=true, ApplyPositionFalloff=false)
   &ScaleColor(
    Scale Alpha=0.7
    Scale RGB=&VectorFromFloat(Value=&FNTimeOfDayScalar(Night Scale=5, Morning Scale=5, Day Scale=5, Evening Scale=5))
    ScaleA=false
    ScaleRGB=true
   ) {module=ScaleColor001}
   &Drag(Drag=5, Use Linear Drag=true, Use Rotational Drag=false)
   &SolveForcesAndVelocity(Clamp Velocity=false, Limit Acceleration=false)
   /SetVariables(Particles.EnergyColor:LinearColor=/DealershipSharedLibrary/FX/Niagara/DynamicInputs/MultiplyLinearColorByVectorAndFloat(
     Color to Scale:LinearColor=&ColorFromCurve(ColorCurve=DataInterface() {propertyValues {
        AlphaCurve {keys [<at 0 0.5 RCIM_Cubic>,<at 0.9052 0 RCIM_Cubic>]}
        BlueCurve {keys [<at 0 1 RCIM_Cubic>,<at 0.2155 1 RCIM_Cubic>]}
        GreenCurve {keys [<at 0 0.7939 RCIM_Cubic>,<at 0.2155 0.6028 RCIM_Cubic>]}
        RedCurve {keys [<at 0 0.6135 RCIM_Cubic>,<at 0.2155 0.2552 RCIM_Cubic>]}
       }})
     RGB Scale:Vector3f=<1 1 1>
     Alpha Scale:NiagaraFloat=&FloatFromCurve(
      FloatCurve=DataInterface() {propertyValues {Curve {keys [{arriveTangent=-3.0806, interpMode=RCIM_Cubic, leaveTangent=-3.0806, tangentMode=RCTM_User, time=0, value=1}, <at 0.85 0.2071 RCIM_Cubic>]}}}
      Scale Curve=&FNTimeOfDayScalar(Night Scale=0.75, Morning Scale=0.75, Day Scale=0.75, Evening Scale=0.75)
     )
    )) {isSetParameters=true}
   /SetVariables(
    Particles.SnowWidth:NiagaraFloat=&FloatFromCurve(FloatCurve=DataInterface() {propertyValues {Curve {keys [<at 0 1>,<at 1 0>]}}}, Scale Curve=12)
    Particles.snow filler width:NiagaraFloat=&FloatFromCurve(FloatCurve=DataInterface() {propertyValues {Curve {keys [<at 0 1>,<at 1 0>]}}}, Scale Curve=15)
   ) {isSetParameters=true}
   /SetVariables(Particles.SnowColor:LinearColor=&MakeLinearColorFromVectorAndFloat(
     "Vector (RGB)"=&MakeVectorFromLinearColorRGB(Color=&ColorFromCurve(ColorCurve=DataInterface() {propertyValues {
         AlphaCurve {keys [<at 0 0.5 RCIM_Cubic>,<at 0.9052 0 RCIM_Cubic>]}
         BlueCurve {keys [<at 0 5 RCIM_Cubic>,<at 0.1985 1 RCIM_Cubic>]}
         GreenCurve {keys [<at 0 5 RCIM_Cubic>,<at 0.1985 0.8023 RCIM_Cubic>]}
         RedCurve {keys [<at 0 5 RCIM_Cubic>,<at 0.1985 0.5438 RCIM_Cubic>]}
        }}))
     "Float (Alpha)"=&FloatFromCurve(
      FloatCurve=DataInterface() {propertyValues {Curve {keys [<at 0 1>,<at 1 0>]}}}
      Scale Curve=&FNTimeOfDayScalar(Night Scale=1, Morning Scale=1, Evening Scale=1)
     )
    )) {isSetParameters=true}
  ]
  rendererClasses [&NiagaraRibbonRendererProperties]
  renderers [
   {rendererClass=&NiagaraRibbonRendererProperties}
   {rendererClass=&NiagaraRibbonRendererProperties, rendererIndex=1}
   {rendererClass=&NiagaraRibbonRendererProperties, rendererIndex=2}
  ]
 }
 moment_glow_and_flow_impulse {
  emitterUpdateScript [
   &EmitterState(Life Cycle Mode=::NewEnumerator1, Loop Behavior=::NewEnumerator1, Loop Delay=1.35, Loop Duration=1, UseLoopDelay=true)
   &SpawnBurst_Instantaneous(Use Spawn Probability=false, UseLoopCountLimit=false)
  ]
  particleSpawnScript [
   /SetVariables(Particles.VisibilityTag=&SpawnGroup()) {isSetParameters=true}
   &InitializeParticle(
    Lifetime=&MergeHelper_FloatSwitch(0=0.3, 1=0.2, 2=?, 3=?, Index=&SpawnGroup(), Num=::NewEnumerator0)
    Mass Mode=::NewEnumerator1
    Mesh Scale Mode=::NewEnumerator3
    Mesh Uniform Scale=&FloatFromCurve(
     FloatCurve=DataInterface() {propertyValues {Curve {keys [{tangentMode=RCTM_User, time=0, value=1}, {tangentMode=RCTM_User, time=1, value=2}]}}}
     CurveIndex=&ReturnNormalizedExecIndex()
     Scale Curve=5
    )
    Position Mode=::NewEnumerator1
    Sprite Size=<300 500>
    Sprite Size Mode=::NewEnumerator4
    UsePositionOffset=false
    Write Material Random=true
   ) {module=InitializeParticle001}
   &SkeletalMeshLocation(
    Inherit Velocity=false
    MaskSpawnGroup=false
    Mesh Sampling Type=::NewEnumerator5
    Skeletal Mesh=$System.Skeletal Mesh
    Socket Sampling Mode=::NewEnumerator5
   )
   &CameraOffset(Camera Offset Amount=-15)
   &InitialMeshOrientation(Mesh Orientation Mode=::NewEnumerator2, Orientation Vector=<0 0 1>, Rotation Coordinate Space=::NewEnumerator2, Use Rotation Vector=true)
   /SetVariables(
    Particles.feet SpriteFacing:Vector3f=<0 0 1>
    Particles.feet sprite Position:NiagaraPosition=&AddVectorToPosition(Position=$Particles.Position, Vector=<0 0 12>)
   ) {isSetParameters=true}
  ]
  particleUpdateScript [
   &ParticleState()
   &SolveForcesAndVelocity(Clamp Velocity=false, Limit Acceleration=false)
   &ScaleColor(
    Scale Alpha=&FloatFromCurve(FloatCurve=DataInterface() {propertyValues {Curve {keys [
         {arriveTangent=6.4019, interpMode=RCIM_Cubic, leaveTangent=6.4019, tangentMode=RCTM_User, time=0, value=0}
         {arriveTangent=0.002, interpMode=RCIM_Cubic, leaveTangent=0.002, time=0.2, value=0.7}
         {arriveTangent=-0.9006, interpMode=RCIM_Cubic, leaveTangent=-0.9006, tangentMode=RCTM_User, time=1.0011, value=0.002}
        ]}}})
    Scale RGB=&MakeVectorFromLinearColorRGB(Color=&FNTimeOfDayColor(Night Color=<rgba 0.15 0.7167 1 1>, Morning Color=<rgba 1 1 1 1>, Day Color=<rgba 1 1 1 1>, Evening Color=<rgba 1 1 1 1>))
    ScaleA=true
    ScaleRGB=true
   )
   &ScaleSpriteSize(Scale Factor=&Multiply_Vector2D_ByFloat(
     Float=&FloatFromCurve(
      FloatCurve=DataInterface() {propertyValues {Curve {keys [
          {arriveTangent=0.7585, interpMode=RCIM_Cubic, leaveTangent=0.7585, tangentMode=RCTM_User, time=0, value=0.4}
          {interpMode=RCIM_Cubic, tangentMode=RCTM_User, time=1, value=1}
         ]}}}
      Scale Curve=1.25
     )
     Vector2D=<1 1>
    ))
   &ScaleMeshSize(Scale Factor=&VectorFromFloat(Value=&FloatFromCurve(
      FloatCurve=DataInterface() {propertyValues {Curve {keys [
          {arriveTangent=4.1785, interpMode=RCIM_Cubic, leaveTangent=4.1785, tangentMode=RCTM_User, time=0, value=0}
          {arriveTangent=1.1179, interpMode=RCIM_Cubic, leaveTangent=1.1179, tangentMode=RCTM_User, time=0.2042, value=0.6228}
          {arriveTangent=0.0224, interpMode=RCIM_Cubic, leaveTangent=0.0224, tangentMode=RCTM_User, time=0.5, value=1}
         ]}}}
      Scale Curve=0.12
     )))
   /SetVariables(
    Particles.DynamicParameter_01:Vector4f=&MakeVector4(
     X=&RandomRangeFloat(Minimum=0, Maximum=1)
     Y=2.5
     Z=&RandomRangeFloat(Minimum=2, Maximum=3)
     W=0
    )
    Particles.DynamicParameter_02:Vector4f=&MakeVector4(X=0, Y=&RandomRangeFloat(Minimum=0, Maximum=1), Z=0, W=0)
   ) {isSetParameters=true}
   /SetVariables(Particles.Corona_Size:Vector2f=&Multiply_Vector2D_ByFloat(
     Float=&FloatFromCurve(
      FloatCurve=DataInterface() {propertyValues {Curve {defaultValue=1, keys [{tangentMode=RCTM_User, time=0, value=0}, {tangentMode=RCTM_User, time=1, value=1}]}}}
      Scale Curve=2
     )
     Vector2D=$Particles.SpriteSize
    )) {isSetParameters=true}
   /SetVariables(Particles.DarkGlowSize:Vector2f=&Multiply_Vector2D_ByFloat(Float=0.75, Vector2D=$Particles.SpriteSize)) {isSetParameters=true}
   /SetVariables(Particles.CentralFlareSize:Vector2f=&Multiply_Vector2D_ByFloat(Float=3, Vector2D=$Particles.SpriteSize)) {isSetParameters=true}
   /SetVariables(Particles.CentralFlareColor:LinearColor=&MakeLinearColorFromVectorAndFloat(
     "Vector (RGB)"=&MakeVectorFromLinearColorRGB(Color=&ColorFromCurve(ColorCurve=DataInterface() {propertyValues {
         AlphaCurve {keys [<at 0 1>,<at 0.8001 0>]}
         BlueCurve {keys [<at 0 1>,<at 0.5062 0.9905>]}
         GreenCurve {keys [<at 0 1>,<at 0.5062 1>]}
         RedCurve {keys [<at 0 1>,<at 0.5062 0.6419>]}
        }}))
     "Float (Alpha)"=&FloatFromCurve(
      FloatCurve=DataInterface() {propertyValues {Curve {
         defaultValue=1
         keys [
          {arriveTangent=9.0725, interpMode=RCIM_Cubic, leaveTangent=9.0725, tangentMode=RCTM_User, time=0, value=0}
          {arriveTangent=-2.5249, interpMode=RCIM_Cubic, leaveTangent=-2.5249, tangentMode=RCTM_User, time=0.1408, value=0.9847}
          {arriveTangent=-0.0405, interpMode=RCIM_Cubic, leaveTangent=-0.0405, tangentMode=RCTM_User, time=1, value=0}
         ]
        }}}
      Scale Curve=&FN_FrontEndFloatModifier(
       Gameplay=&FNTimeOfDayScalar(Night Scale=0.05, Morning Scale=0.05, Day Scale=0.05, Evening Scale=0.05)
       Front End=0.05
      )
     )
    )) {isSetParameters=true}
   &DynamicMaterialParameters(Module.Index0Float0Write=true, Module.Index0Float1Write=true, Module.Index0Float2Write=true, Module.Index0Float3Write=true)
   /SetVariables(Particles.Body Glow position:NiagaraPosition=&AddVectorToPosition(Position=$Particles.Position, Vector=<0 0 85>)) {isSetParameters=true}
  ]
  rendererClasses [&NiagaraSpriteRendererProperties]
  renderers [{rendererClass=&NiagaraSpriteRendererProperties}]
 }
}
summary {
 systemName=NS_Emote_IcicleCane_FrostyHands
 emitters [
  {emitterName=Frosty_Mist_hands, rendererClasses [&NiagaraSpriteRendererProperties]}
  {emitterName=hand_trails, rendererClasses [&NiagaraRibbonRendererProperties]}
  {emitterName=moment_glow_and_flow_impulse, rendererClasses [&NiagaraSpriteRendererProperties]}
  {emitterName=Frosty_Mist_body_Burst, rendererClasses [&NiagaraSpriteRendererProperties]}
  {
   emitterName=Frost_Impact_ground002
   rendererClasses [&NiagaraSpriteRendererProperties, &NiagaraMeshRendererProperties]
  }
  {emitterName=Hand_Sparkles, rendererClasses [&NiagaraSpriteRendererProperties]}
  {emitterName=Hand_flake_off, rendererClasses [&NiagaraSpriteRendererProperties]}
  {emitterName=Frozen_mist_Burst, rendererClasses [&NiagaraSpriteRendererProperties]}
  {emitterName=Ground_burst, rendererClasses [&NiagaraSpriteRendererProperties]}
 ]
 userVariables [
  User.Anim_Curve_SpawnRate_Ribbon {type=NiagaraFloat, defaultValue=0, description=""}
  User.AnimTrail_SpawnRate {type=NiagaraFloat, defaultValue=0, description=""}
  User.bIsFrontEnd {type=NiagaraBool}
  User.bIsFrontEndPreview {type=NiagaraBool}
  User.FrostyMistAlpha {type=NiagaraFloat, defaultValue=0, description=""}
  User.FrostyMistTrail_Rate {type=NiagaraFloat, defaultValue=1, description=""}
  User.MistyTrail_Life {type=NiagaraFloat, defaultValue=0, description=""}
  User.RibbonIDOffset {type=NiagaraInt32}
 ]
}
